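# Niagara effect for the Void Redemption gadget that draws blue ribbon trails and glows attached to a skeletal mesh, on the hands going by the asset name.
# FN_AttachedRibbon spawns two short-lived, skeletal-positioned ribbons that taper over their lifetime and use three ribbon renderers.
# Light bursts two particles, pins each to a bone or socket by spawn group index, and renders layered sprites plus a curve-scaled light.
# Both emitters read the system-level skeletal mesh.
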NiagaraSystem /BRCosmetics/Animation/Props/Gadgets_VoidRedemption/FX/NS_VoidRedemtion_RibbonsHands
dependencies {
 usedDataInterfaces [&NiagaraDataInterfaceCurve]
 usedDynamicInputs []
 usedModules [
  &SystemState
  &self:SystemSpawnScript.SystemScriptSource.SystemScriptGraph.NiagaraNodeAssignment_0.SetVariables_#1
 ]
}
emitters {
 FN_AttachedRibbon {
  emitterUpdateScript [
   &EmitterState()
   &FN_SpawnRibbon_v2(Number of Ribbons=2, Spawn Mode=::NewEnumerator0, SpawnRate=30)
   &FN_Integrator() {enabled=false}
  ]
  particleSpawnScript [
   /Game/Effects/Niagara/Modules/Ribbons/FN_InitializeRibbon_v4(
    Color:LinearColor=<rgba 0 7.24 25 1>
    Enable Skeletal Position=true
    Lifetime=0.1
    Ribbon Width=8
    SkeletalMesh:NiagaraDataInterfaceSkeletalMesh=$System.Skeletal Mesh
   )
   &FN_FlipRibbonV()
   &AddVelocity(Velocity=<0 0 100>) {enabled=false}
  ]
  particleUpdateScript [
   &SolveForcesAndVelocity(Clamp Velocity=false, Limit Acceleration=false)
   /Game/Effects/Niagara/Modules/Ribbons/FN_UpdateRibbon_v4(Enable Skeletal Position=true, SkeletalMesh:NiagaraDataInterfaceSkeletalMesh=$System.Skeletal Mesh)
   &FN_RibbonMaterial() {enabled=false}
   &ScaleRibbonWidth(Ribbon Width Scale=&FloatFromCurve(FloatCurve=DataInterface() {propertyValues {Curve {keys [
         {interpMode=RCIM_Cubic, tangentMode=RCTM_User, time=0, value=1}
         {arriveTangent=-2.4019, interpMode=RCIM_Cubic, leaveTangent=-2.4019, tangentMode=RCTM_User, time=1, value=0}
        ]}}}))
   &DynamicMaterialParameters(
    Index 0 Param 1=1
    Index 0 Param 2=1
    Index 0 Param 3=1
    Index 0 Param 4=1
    Index 1 Param 1=1
    Index 1 Param 2=1
    Index 1 Param 3=1
    Index 1 Param 4=1
    Index 3 Param 3=1
    Index 3 Param 4=1
    Module.Index0Param0Write=true
    Module.Index0Param1Write=true
    Module.Index0Param2Write=true
    Module.Index0Param3Write=true
    Module.Index1Param0Write=true
    Module.Index1Param1Write=true
    Module.Index1Param2Write=true
    Module.Index1Param3Write=true
    Module.Index2Param0Write=true
    Module.Index2Param1Write=true
    Module.Index2Param2Write=true
    Module.Index2Param3Write=true
    Module.Index3Param0Write=true
    Module.Index3Param1Write=true
    Module.Index3Param2Write=true
    Module.Index3Param3Write=true
    Write Parameter Index 1=true
    Write Parameter Index 2=true
    Write Parameter Index 3=true
   ) {enabled=false}
  ]
  rendererClasses [&NiagaraRibbonRendererProperties]
  renderers [
   {rendererClass=&NiagaraRibbonRendererProperties}
   {rendererClass=&NiagaraRibbonRendererProperties, rendererIndex=1}
   {rendererClass=&NiagaraRibbonRendererProperties, rendererIndex=2}
  ]
 }
 Light {
  emitterUpdateScript [
   &EmitterState(Inactive Response=::NewEnumerator1, Life Cycle Mode=::NewEnumerator1, Loop Behavior=::NewEnumerator1, Loop Duration=2, UseLoopDelay=false)
   &SpawnBurst_Instantaneous(Spawn Count=2, Use Spawn Probability=false, UseLoopCountLimit=false)
  ]
  particleSpawnScript [
   &InitializeParticle(
    Color=<rgba 0.2 2.4589 8 0.07>
    Sprite Size Mode=::NewEnumerator3
    Uniform Sprite Size=75
    UsePositionOffset=false
    Write Material Random=true
   )
   &FN_SetSpawnGroup(Number of Spawn Groups=2)
   /SetVariables(
    Particles.R2 Color:LinearColor=<rgba 0 14.48 50 1>
    Particles.R2 Size:Vector2f=&Vector2DFromFloat(Value=7)
   ) {isSetParameters=true}
  ]
  particleUpdateScript [
   &ParticleState()
   &SkeletalMeshLocation(Bone / Socket Index=$Particles.SpawnGroup, Bone Sampling Mode=::NewEnumerator4, MaskSpawnGroup=false, Skeletal Mesh=$System.Skeletal Mesh)
   &ScaleSpriteSize(Uniform Curve Sprite Scale=DataInterface() {propertyValues {Curve {keys [
        {arriveTangent=2.8688, interpMode=RCIM_Cubic, leaveTangent=2.8688, tangentMode=RCTM_User, time=0, value=0}
        {interpMode=RCIM_Cubic, tangentMode=RCTM_User, time=1, value=1}
       ]}}}) {enabled=false}
   &ScaleColor(
    Scale Alpha=&FloatFromCurve(FloatCurve=DataInterface() {propertyValues {Curve {keys [{interpMode=RCIM_Cubic, tangentMode=RCTM_User, time=0, value=1}, {interpMode=RCIM_Cubic, tangentMode=RCTM_User, time=1, value=0}]}}})
    ScaleA=true
    ScaleRGB=true
   ) {enabled=false}
   &FN_Light_Attributes(
    Color=$Particles.Color
    Create Color Attribute=true
    Create Exponent Attribute=true
    Create Radius Attribute=true
    Exponent=5
    Radius=120
    Scale Color=&FloatFromCurve(
     FloatCurve=DataInterface() {propertyValues {Curve {keys [
         {interpMode=RCIM_Cubic, tangentMode=RCTM_User, time=0, value=1}
         {arriveTangent=-3.4836, interpMode=RCIM_Cubic, leaveTangent=-3.4836, tangentMode=RCTM_User, time=1, value=0}
        ]}}}
     Scale Curve=2
    )
   )
   &CameraOffset(Camera Offset Amount=10)
  ]
  rendererClasses [&NiagaraSpriteRendererProperties, &NiagaraLightRendererProperties]
  renderers [
   {rendererClass=&NiagaraSpriteRendererProperties}
   {rendererClass=&NiagaraLightRendererProperties, rendererIndex=1}
   {rendererClass=&NiagaraSpriteRendererProperties, rendererIndex=2}
  ]
 }
}
summary {
 systemName=NS_VoidRedemtion_RibbonsHands
 emitters [
  {emitterName=FN_AttachedRibbon, rendererClasses [&NiagaraRibbonRendererProperties]}
  {emitterName=Light, rendererClasses [&NiagaraSpriteRendererProperties, &NiagaraLightRendererProperties]}
 ]
 userVariables [
  User.RibbonIDOffset {type=NiagaraInt32}
  User.SkeletalMesh {
   type=NiagaraDataInterfaceSkeletalMesh
   defaultValue {
    struct=NiagaraExt_VariableValue_DataInterface
    value {dataInterface=&self:NiagaraDataInterfaceSkeletalMesh_0}
   }
  }
 ]
}
